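# Niagara cosmetic effect that dresses the Madame Moth Posh NPC skin in red fire flecks, smoke and glow, attached to skeletal mesh bones.
# Owner velocity switches two-sided bone emitters between idle flecks and distance-spawned moving flecks with smoke.
# Head flecks spawn continuously, and two red lights follow the calves.
# Sprites fade with camera distance, some values differ in the front end, and the Fire_spline mesh emitter is disabled.
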NiagaraSystem /BRCosmetics/Characters/Player/Female/Medium/Bodies/F_MED_MadameMoth_Posh_NPC/FX/NS_NPC_MadameMoth_Skin_Posh
dependencies {
 usedDataInterfaces [
  &NiagaraDataInterfaceArrayFloat3
  &NiagaraDataInterfaceCurve
  &NiagaraDataInterfaceSpriteRendererInfo
  &NiagaraDataInterfaceColorCurve
  &NiagaraDataInterfaceArrayFloat
  &NiagaraDataInterfacePlatformSet
  &NiagaraDataInterfaceSkeletalMesh
 ]
 usedDynamicInputs []
 usedModules [
  &self:SystemSpawnScript.SystemScriptSource.SystemScriptGraph.NiagaraNodeAssignment_3.SetVariables_#1
  &self:SystemSpawnScript.SystemScriptSource.SystemScriptGraph.NiagaraNodeAssignment_6.SetVariables_#2
  &self:SystemSpawnScript.SystemScriptSource.SystemScriptGraph.NiagaraNodeAssignment_1.SetVariables_#3
  &self:SystemSpawnScript.SystemScriptSource.SystemScriptGraph.NiagaraNodeAssignment_0.SetVariables_#4
  &self:SystemSpawnScript.SystemScriptSource.SystemScriptGraph.NiagaraNodeAssignment_4.SetVariables_#5
  &self:SystemSpawnScript.SystemScriptSource.SystemScriptGraph.NiagaraNodeAssignment_5.SetVariables_#6
  &SystemState
  &self:SystemSpawnScript.SystemScriptSource.SystemScriptGraph.NiagaraNodeAssignment_2.SetVariables_#7
 ]
}
emitters {
 Fire_pieces_Idle {
  emitterUpdateScript [
   &EmitterState(Inactive Response=::NewEnumerator1, Life Cycle Mode=::NewEnumerator1, Loop Delay=2.9, Loop Duration=1, UseLoopDelay=false)
   &SpawnRate(
    SpawnRate=&Multiply_Float(
     A=14
     B=&MakeCustomFloatFromBool(
      Bool=&SetBoolByFloatComparison(A=&VectorLength(Vector=$Engine.Owner.Velocity), B=10)
      True Float=0
      False Float=1
     )
    )
    Use Spawn Probability=false
   )
  ]
  particleSpawnScript [
   &PartitionParticles()
   /SetVariables(Particles.Random size:NiagaraFloat=&RandomRangeFloat(Minimum=0.2, Maximum=0.25)) {isSetParameters=true, module=SetVariables_#8}
   &InitializeParticle(
    Color=<rgba 5 0 0 1>
    Lifetime=&Multiply_Float(A=&RandomRangeFloat(Minimum=0.5, Maximum=2), B=1)
    Sprite Rotation Mode=::NewEnumerator2
    Sprite Size Mode=::NewEnumerator3
    Sprite UV Mode=::NewEnumerator3
    Uniform Sprite Size=&Multiply_Float(A=100, B=$Particles.Random size)
    UsePositionOffset=false
    Write Material Random=true
   ) {module=InitializeParticle001}
   &DynamicMaterialParameters(
    Index 0 Param 2=&RandomRangeFloat(Minimum=0, Maximum=4)
    Index 0 Param 4=2.5
    Module.Param0WriteEnabled=true
    Module.Param1WriteEnabled=true
    Module.Param2WriteEnabled=false
    Module.Param3WriteEnabled=true
   ) {module=DynamicMaterialParameters001}
   &ShapeLocation(
    MaskSpawnGroup=false
    Override Seed=false
    Shape Origin=&VectorFromFloat(Value=0)
    Sphere Radius=3
    UseOwnerScale=true
   )
   &AddVelocity(Velocity=&SelectVectorFromArray(
     Direct Array Index=$Particles.PartitionParticles.Partition
     Vector Selection Array=DataInterface() {propertyValues {FloatData [<-10 0 0>, <10 0 0>]}}
    )) {module=AddVelocity001}
   &AddVelocity(Rotation Coordinate Space=::World, Velocity=<0 0 60>)
   /SetVariables() {isSetParameters=true, module=SetVariables_#9}
   /SetVariables(Particles.DY3:Vector4f=&MakeVector4(X=3, Y=1, Z=1, W=1)) {isSetParameters=true, module=SetVariables_#10}
  ]
  particleUpdateScript [
   &SkeletalMeshLocation(
    Bone / Socket Index=$Particles.PartitionParticles.Partition
    Bone Sampling Mode=::NewEnumerator4
    MaskSpawnGroup=false
    Position Sampling=::NewEnumerator1
    Sampled Position Offset=&AddVectorToPosition(
     Position=$Particles.ShapeLocation.ShapePosition
     Vector=&SelectVectorFromArray(
      Direct Array Index=$Particles.PartitionParticles.Partition
      Vector Selection Array=DataInterface() {propertyValues {FloatData [<15 0 0>, <15 0 0>]}}
     )
    )
    Skeletal Mesh=$System.Skeletal Mesh two side
    Source Movement Scale:Vector3f=<0 0 0>
   )
   /SetVariables(Particles.Size Scale:NiagaraFloat=&FloatFromCurve(
     FloatCurve=DataInterface() {propertyValues {Curve {keys [
         {interpMode=RCIM_Cubic, tangentMode=RCTM_User, time=0, value=0}
         {interpMode=RCIM_Cubic, tangentMode=RCTM_User, time=0.5, value=0.3}
         {tangentMode=RCTM_User, time=1, value=0}
        ]}}}
     Scale Curve=0.7
    )) {isSetParameters=true, module=SetVariables_#11}
   &SubUVAnimation(
    End Frame Range Override=11
    "Sprite Renderer "=DataInterface() {propertyValues {SpriteRenderer=&self:Fire_pieces_Idle_9.NiagaraSpriteRendererProperties_0}}
    Start Frame Range Override=8
    SubUV Animation Mode:ENiagara_SubUVLookupModeV2=::NewEnumerator2
    UseEndFrame=true
    UseStartFrame=true
   ) {module=SubUVAnimation001}
   /Niagara/Modules/Update/Forces/CurlNoiseForce(Noise Frequency=100, Noise Strength=&Multiply_Float(A=100, B=1), Pan Noise=true)
   &ParticleState()
   &ScaleSpriteSize(Scale Sprite Size Mode=::NewEnumerator0, Uniform Scale Factor=$Particles.Size Scale) {module=ScaleSpriteSize001}
   &SpriteRotationRate(Rotation Rate=&RandomRangeFloat(Minimum=-5, Maximum=5)) {module=SpriteRotationRate001}
   &Drag(Use Linear Drag=true, Use Rotational Drag=true)
   &SolveForcesAndVelocity(Clamp Velocity=false, Limit Acceleration=false)
   &ScaleColor(
    Scale Alpha=&Multiply_Float(
     A=&FloatFromCurve(
      FloatCurve=DataInterface() {propertyValues {Curve {keys [
          {interpMode=RCIM_Cubic, tangentMode=RCTM_User, time=0, value=0}
          <at 0.4 0>
          {interpMode=RCIM_Cubic, tangentMode=RCTM_User, time=0.5, value=0.75}
          {tangentMode=RCTM_User, time=1, value=0}
         ]}}}
      Scale Curve=&MakeCustomFloatFromBool(
       Bool=&SetBoolByFloatComparison(A=&VectorLength(Vector=$Particles.SkeletalMeshLocation.SampledVelocity), B=10)
       True Float=0
       False Float=1
      )
     )
     B=&FN_LOD_Distance_Remap(Start Distance=1400, End Distance=1700, Minimum Output=1, Maximum Output=0, Clamp Output=true)
    )
    Scale RGB=&Multiply_VectorByFloat(
     Vector=&MakeVectorFromLinearColorRGB(Color=&ColorFromCurve(ColorCurve=DataInterface() {propertyValues {
         AlphaCurve {keys []}
         BlueCurve {keys [<at 0 1 RCIM_Cubic>,<at 0.3533 0>,<at 0.6474 0>]}
         GreenCurve {keys [<at 0 1 RCIM_Cubic>,<at 0.3533 0>,<at 0.6474 0>]}
         RedCurve {keys [<at 0 1 RCIM_Cubic>,<at 0.3533 0.25>,<at 0.6474 0.1>]}
        }}))
     Float=0.5
    )
    ScaleA=true
    ScaleRGB=true
   )
   &FN_Sprite_Attributes(
    Color=&MakeLinearColorFromVectorAndFloat(
     "Vector (RGB)"=&MakeVectorFromLinearColorRGB(Color=&ColorFromCurve(ColorCurve=DataInterface() {propertyValues {
         AlphaCurve {keys []}
         BlueCurve {keys [<at 0 0.82 RCIM_Cubic>,<at 1 0.1 RCIM_Cubic>]}
         GreenCurve {keys [<at 0 0.6758 RCIM_Cubic>,<at 1 0 RCIM_Cubic>]}
         RedCurve {keys [<at 0 5 RCIM_Cubic>,<at 1 3 RCIM_Cubic>]}
        }}))
     "Float (Alpha)"=&Multiply_Float(
      A=2
      B=&FN_LOD_Distance_Remap(Start Distance=1400, End Distance=1700, Minimum Output=1, Maximum Output=0, Clamp Output=true)
     )
    )
    Create Color Attribute=true
    Create SpriteSize Attribute=true
    SpriteSize=&Vector2DFromFloat(Value=&Multiply_Float(A=60, B=&Multiply_Float(A=$Particles.Random size, B=$Particles.Size Scale)))
    Write to Module Namespace=true
   )
   &FN_Sprite_Attributes(
    Color=&MakeLinearColorFromVectorAndFloat(
     "Vector (RGB)"=&MakeVectorFromLinearColorRGB(Color=<rgba 1.5 0 0.05 0.5>)
     "Float (Alpha)"=&Multiply_Float(
      A=&FN_FrontEndFloatModifier(Gameplay=0.2, Front End=0.2)
      B=&FN_LOD_Distance_Remap(Start Distance=1400, End Distance=1700, Minimum Output=1, Maximum Output=0, Clamp Output=true)
     )
    )
    Create Color Attribute=true
    Create SpriteSize Attribute=true
    SpriteSize=&Vector2DFromFloat(Value=&Multiply_Float(A=180, B=&Multiply_Float(A=$Particles.Random size, B=$Particles.Size Scale)))
    Write to Module Namespace=true
   ) {module=FN_Sprite_Attributes001}
  ]
  rendererClasses [&NiagaraSpriteRendererProperties]
  renderers [
   {rendererClass=&NiagaraSpriteRendererProperties}
   {rendererClass=&NiagaraSpriteRendererProperties, rendererIndex=1}
   {rendererClass=&NiagaraSpriteRendererProperties, rendererIndex=2}
  ]
 }
 Fire_pieces_Idle001 {
  emitterUpdateScript [
   &EmitterState(Inactive Response=::NewEnumerator1, Life Cycle Mode=::NewEnumerator1, Loop Delay=2.9, Loop Duration=1, UseLoopDelay=false)
   &SpawnPerUnit(
    Max Movement Threshold=1000
    Module.Use Max Movement Threshold=true
    Module.Use Movement Tolerance=false
    Spawn Spacing:NiagaraFloat=&Multiply_Float(
     A=25
     B=&MakeCustomFloatFromBool(
      Bool=&SetBoolByFloatComparison(A=&VectorLength(Vector=$Engine.Owner.Velocity), B=100)
      True Float=1
      False Float=0
     )
    )
    Use Spawn Probability=false
   )
  ]
  particleSpawnScript [
   &PartitionParticles()
   /SetVariables(Particles.Random size:NiagaraFloat=&RandomRangeFloat(Minimum=0.5, Maximum=0.7)) {isSetParameters=true, module=SetVariables_#8}
   &InitializeParticle(
    Color=<rgba 5 0 0 2>
    Lifetime Max=0.6
    Lifetime Min=0.4
    Lifetime Mode=::NewEnumerator1
    Sprite Rotation Mode=::NewEnumerator2
    Sprite Size Mode=::NewEnumerator3
    Sprite UV Mode=::NewEnumerator3
    Uniform Sprite Size=&Multiply_Float(A=100, B=$Particles.Random size)
    UsePositionOffset=false
    Write Material Random=true
   ) {module=InitializeParticle001}
   &DynamicMaterialParameters(
    Index 0 Param 1=5
    Index 0 Param 2=&RandomRangeFloat(Minimum=0, Maximum=4)
    Index 0 Param 4=2.5
    Module.Param0WriteEnabled=true
    Module.Param1WriteEnabled=true
    Module.Param2WriteEnabled=false
    Module.Param3WriteEnabled=true
   ) {module=DynamicMaterialParameters001}
   &ShapeLocation(
    MaskSpawnGroup=false
    Override Seed=false
    Shape Origin=&VectorFromFloat(Value=0)
    Sphere Radius=5
    UseOwnerScale=true
   )
   /SetVariables() {isSetParameters=true, module=SetVariables_#9}
   /SetVariables(Particles.DY3:Vector4f=&MakeVector4(X=3, Y=1, Z=1, W=1)) {isSetParameters=true, module=SetVariables_#10}
   &CameraOffset(Camera Offset Amount=1)
  ]
  particleUpdateScript [
   &SkeletalMeshLocation(
    Bone / Socket Index=$Particles.PartitionParticles.Partition
    Bone Sampling Mode=::NewEnumerator4
    MaskSpawnGroup=false
    Position Sampling=::NewEnumerator1
    Sampled Position Offset=&AddVectorToPosition(
     Position=$Particles.ShapeLocation.ShapePosition
     Vector=&SelectVectorFromArray(
      Direct Array Index=$Particles.PartitionParticles.Partition
      Vector Selection Array=DataInterface() {propertyValues {FloatData [<0 0 0>, <0 0 0>]}}
     )
    )
    Skeletal Mesh=$System.Skeletal Mesh two side
    Source Movement Scale:Vector3f=<0.7 0.7 0.7>
   )
   /SetVariables(Particles.Size Scale:NiagaraFloat=&FloatFromCurve(
     FloatCurve=DataInterface() {propertyValues {Curve {keys [{interpMode=RCIM_Cubic, tangentMode=RCTM_User, time=0, value=0}, <at 0.2 0.7>, {tangentMode=RCTM_User, time=1, value=0}]}}}
     Scale Curve=0.7
    )) {isSetParameters=true, module=SetVariables_#11}
   &SubUVAnimation(
    End Frame Range Override=3
    Evaluation Type=::NewEnumerator0
    "Sprite Renderer "=DataInterface() {propertyValues {SpriteRenderer=&self:Fire_pieces_Idle001_6.NiagaraSpriteRendererProperties_1}}
    SubUV Animation Mode:ENiagara_SubUVLookupModeV2=::NewEnumerator2
    UseEndFrame=true
    UseStartFrame=true
   )
   &CurlNoiseForce(Noise Frequency=100, Noise Strength=&Multiply_Float(A=100, B=1), Pan Noise=true)
   &ParticleState()
   &ScaleSpriteSize(Scale Sprite Size Mode=::NewEnumerator0, Uniform Scale Factor=$Particles.Size Scale) {module=ScaleSpriteSize001}
   &SpriteRotationRate(Rotation Rate=&RandomRangeFloat(Minimum=-400, Maximum=400)) {module=SpriteRotationRate001}
   &Drag(Use Linear Drag=true, Use Rotational Drag=true)
   &SolveForcesAndVelocity(Clamp Velocity=false, Limit Acceleration=false)
   &ScaleColor(
    Scale Alpha=&Multiply_Float(
     A=&FloatFromCurve(FloatCurve=DataInterface() {propertyValues {Curve {keys [
          {interpMode=RCIM_Cubic, tangentMode=RCTM_User, time=0, value=0}
          <at 0.4 0>
          {interpMode=RCIM_Cubic, tangentMode=RCTM_User, time=0.5, value=0.75}
          {tangentMode=RCTM_User, time=1, value=0}
         ]}}})
     B=&FN_LOD_Distance_Remap(Start Distance=1400, End Distance=1700, Minimum Output=1, Maximum Output=0, Clamp Output=true)
    )
    Scale RGB=&Multiply_VectorByFloat(
     Vector=&MakeVectorFromLinearColorRGB(Color=&ColorFromCurve(ColorCurve=DataInterface() {propertyValues {
         AlphaCurve {keys []}
         BlueCurve {keys [<at 0 1 RCIM_Cubic>,<at 0.3533 0>,<at 0.6474 0>]}
         GreenCurve {keys [<at 0 1 RCIM_Cubic>,<at 0.3533 0>,<at 0.6474 0>]}
         RedCurve {keys [<at 0 1 RCIM_Cubic>,<at 0.3533 0.25>,<at 0.6474 0.1>]}
        }}))
     Float=0.5
    )
    ScaleA=true
    ScaleRGB=true
   )
   &FN_Sprite_Attributes(
    Color=&MakeLinearColorFromVectorAndFloat(
     "Vector (RGB)"=&MakeVectorFromLinearColorRGB(Color=&ColorFromCurve(ColorCurve=DataInterface() {propertyValues {
         AlphaCurve {keys []}
         BlueCurve {keys [<at 0 0.328 RCIM_Cubic>,<at 1 0.0267>]}
         GreenCurve {keys [<at 0 0.2703 RCIM_Cubic>,<at 1 0>]}
         RedCurve {keys [<at 0 2 RCIM_Cubic>,<at 1 0.5>]}
        }}))
     "Float (Alpha)"=&Multiply_Float(
      A=&FloatFromCurve(FloatCurve=DataInterface() {propertyValues {Curve {keys [
           {tangentMode=RCTM_User, time=0, value=1}
           {arriveTangent=-0.0405, interpMode=RCIM_Cubic, leaveTangent=-0.0405, tangentMode=RCTM_User, time=0.3, value=1}
           {arriveTangent=-1.1636, interpMode=RCIM_Cubic, leaveTangent=-1.1636, tangentMode=RCTM_User, time=1, value=0}
          ]}}})
      B=1
     )
    )
    Create Color Attribute=true
    Create SpriteSize Attribute=true
    SpriteSize=&Vector2DFromFloat(Value=&Multiply_Float(A=80, B=&Multiply_Float(A=$Particles.Random size, B=$Particles.Size Scale)))
    Write to Module Namespace=true
   )
   &FN_Sprite_Attributes(
    Color=&MakeLinearColorFromVectorAndFloat(
     "Vector (RGB)"=&MakeVectorFromLinearColorRGB(Color=<rgba 1.5 0 0.05 0.5>)
     "Float (Alpha)"=&Multiply_Float(A=&FN_FrontEndFloatModifier(Gameplay=0.2, Front End=0.2), B=0.2)
    )
    Create Color Attribute=true
    Create SpriteSize Attribute=true
    SpriteSize=&Vector2DFromFloat(Value=&Multiply_Float(A=180, B=&Multiply_Float(A=$Particles.Random size, B=$Particles.Size Scale)))
    Write to Module Namespace=true
   ) {module=FN_Sprite_Attributes001}
  ]
  rendererClasses [&NiagaraSpriteRendererProperties]
  renderers [
   {rendererClass=&NiagaraSpriteRendererProperties}
   {rendererClass=&NiagaraSpriteRendererProperties, rendererIndex=1}
   {rendererClass=&NiagaraSpriteRendererProperties, rendererIndex=2}
  ]
 }
 Fire_pieces_Idle003 {
  emitterUpdateScript [
   &EmitterState(Inactive Response=::NewEnumerator1, Life Cycle Mode=::NewEnumerator1, Loop Delay=2.9, Loop Duration=1, UseLoopDelay=false)
   &SpawnRate(SpawnRate=12, Use Spawn Probability=false)
  ]
  particleSpawnScript [
   &PartitionParticles(Number Of Partitions=3)
   /SetVariables(Particles.Random size:NiagaraFloat=&RandomRangeFloat(Minimum=0.5, Maximum=0.7)) {isSetParameters=true, module=SetVariables_#8}
   &InitializeParticle(
    Color Mode=::NewEnumerator0
    Lifetime Max=1.5
    Lifetime Mode=::NewEnumerator1
    Sprite Rotation Mode=::NewEnumerator2
    Sprite UV Mode=::NewEnumerator3
    UsePositionOffset=false
    Write Material Random=true
   ) {module=InitializeParticle001}
   &DynamicMaterialParameters(
    Index 0 Param 1=5
    Index 0 Param 2=&RandomRangeFloat(Minimum=0, Maximum=4)
    Index 0 Param 4=2.5
    Module.Param0WriteEnabled=true
    Module.Param1WriteEnabled=true
    Module.Param2WriteEnabled=false
    Module.Param3WriteEnabled=true
   ) {module=DynamicMaterialParameters001}
   /SetVariables() {isSetParameters=true, module=SetVariables_#9}
   /SetVariables(Particles.DY3:Vector4f=&MakeVector4(X=3, Y=1, Z=1, W=1)) {isSetParameters=true, module=SetVariables_#10}
   &CameraOffset()
  ]
  particleUpdateScript [
   /SetVariables(Particles.Size Scale:NiagaraFloat=&FloatFromCurve(
     FloatCurve=DataInterface() {propertyValues {Curve {keys [
         {interpMode=RCIM_Cubic, tangentMode=RCTM_User, time=0, value=0}
         {interpMode=RCIM_Cubic, tangentMode=RCTM_User, time=0.2, value=0.3}
         {tangentMode=RCTM_User, time=1, value=0}
        ]}}}
     Scale Curve=0.7
    )) {isSetParameters=true, module=SetVariables_#11}
   &SkeletalMeshLocation(
    Bone / Socket Index=$Particles.PartitionParticles.Partition
    Bone Sampling Mode=::NewEnumerator4
    MaskSpawnGroup=false
    Sampled Position Offset=&AddVectorToPosition(
     Position=<2 -4 0>
     Vector=&Add_Vector(
      A=&MakeVector(
       X=&Lerp_Float(A=0, B=&RandomRangeFloat(Minimum=5, Maximum=6), Alpha=$Particles.NormalizedAge)
       Y=&Lerp_Float(A=0, B=-8, Alpha=$Particles.NormalizedAge)
       Z=0
      )
      B=&SelectVectorFromArray(
       Direct Array Index=$Particles.PartitionParticles.Partition
       Vector Selection Array=DataInterface() {propertyValues {FloatData [<12 0 7>, <12 0 -7>, <4 -6 0>]}}
      )
     )
    )
    Skeletal Mesh=DataInterface() {propertyValues {
      DefaultMesh=null
      ExcludeBoneName=root
      OptionalComponentClass=null
      PreviewMesh=/BRCosmetics/Characters/Player/Female/Medium/Bodies/F_MED_MadameMoth_Posh_NPC/Meshes/F_MED_MadameMoth_Posh_NPC
      SkinningMode=SkinOnTheFly
      SoftSourceActor=null
      SourceMode=Default
      UvSetIndex=0
      WholeMeshLOD=-1
      bExcludeBone=false
      bReadDeformedGeometry=true
      bRequireCurrentFrameData=true
      ComponentTags []
      FilteredBones [head]
      FilteredSockets []
      MeshUserParameter {parameter {name=None, typeDefHandle {registeredTypeIndex=1}, varData []}}
      SamplingRegions []
     }}
   )
   &SubUVAnimation(
    End Frame Range Override=3
    Evaluation Type=::NewEnumerator0
    "Sprite Renderer "=DataInterface() {propertyValues {SpriteRenderer=&self:Fire_pieces_Idle003_6.NiagaraSpriteRendererProperties_1}}
    SubUV Animation Mode:ENiagara_SubUVLookupModeV2=::NewEnumerator2
    UseEndFrame=true
    UseStartFrame=true
   )
   &ParticleState()
   &SpriteRotationRate(Rotation Rate=&RandomRangeFloat(Minimum=-200, Maximum=200)) {module=SpriteRotationRate001}
   &SolveForcesAndVelocity(Clamp Velocity=false, Limit Acceleration=false)
   &FN_Sprite_Attributes(
    Color=&MakeLinearColorFromVectorAndFloat(
     "Vector (RGB)"=&MakeVectorFromLinearColorRGB(Color=&ColorFromCurve(ColorCurve=DataInterface() {propertyValues {
         AlphaCurve {keys []}
         BlueCurve {keys [<at 0 0.4055 RCIM_Cubic>,<at 0.432 0.107>,<at 0.5945 0.0267>]}
         GreenCurve {keys [<at 0 0.492 RCIM_Cubic>,<at 0.432 0>,<at 0.5945 0>]}
         RedCurve {keys [<at 0 3 RCIM_Cubic>,<at 0.432 2>,<at 0.5945 0.5>]}
        }}))
     "Float (Alpha)"=&Multiply_Float(
      A=&FloatFromCurve(
       FloatCurve=DataInterface() {propertyValues {Curve {keys [
           {tangentMode=RCTM_User, time=0, value=0}
           {arriveTangent=-0.0405, interpMode=RCIM_Cubic, leaveTangent=-0.0405, tangentMode=RCTM_User, time=0.3, value=1}
           {arriveTangent=-1.1636, interpMode=RCIM_Cubic, leaveTangent=-1.1636, tangentMode=RCTM_User, time=1, value=0}
          ]}}}
       Scale Curve=2
      )
      B=&FN_LOD_Distance_Remap(Start Distance=1400, End Distance=1700, Minimum Output=1, Maximum Output=0, Clamp Output=true)
     )
    )
    Create Color Attribute=true
    Create SpriteSize Attribute=true
    SpriteSize=&Vector2DFromFloat(Value=&Multiply_Float(A=80, B=&Multiply_Float(A=$Particles.Random size, B=$Particles.Size Scale)))
    Write to Module Namespace=true
   )
   &FN_Sprite_Attributes(
    Color=&MakeLinearColorFromVectorAndFloat(
     "Vector (RGB)"=&MakeVectorFromLinearColorRGB(Color=<rgba 1.5 0 0.05 0.5>)
     "Float (Alpha)"=&Multiply_Float(
      A=&FN_FrontEndFloatModifier(Gameplay=0.1, Front End=0.2)
      B=&Multiply_Float(
       A=0.2
       B=&FN_LOD_Distance_Remap(Start Distance=1400, End Distance=1700, Minimum Output=1, Maximum Output=0, Clamp Output=true)
      )
     )
    )
    Create Color Attribute=true
    Create SpriteSize Attribute=true
    SpriteSize=&Vector2DFromFloat(Value=&Multiply_Float(A=180, B=&Multiply_Float(A=$Particles.Random size, B=$Particles.Size Scale)))
    Write to Module Namespace=true
   ) {module=FN_Sprite_Attributes001}
   /SetVariables(Particles.Position=&AddVectorToPosition(
     Position=$Particles.Position
     Vector=&Lerp_Vector(A=<0 0 -2>, B=&RandomRangeVector(Minimum=<0 0 6>, Maximum=<0 0 6>), Alpha=$Particles.NormalizedAge)
    )) {isSetParameters=true}
   &ScaleColor(ScaleA=true, ScaleRGB=true)
  ]
  rendererClasses [&NiagaraSpriteRendererProperties]
  renderers [
   {rendererClass=&NiagaraSpriteRendererProperties}
   {rendererClass=&NiagaraSpriteRendererProperties, rendererIndex=1}
   {rendererClass=&NiagaraSpriteRendererProperties, rendererIndex=2}
  ]
 }
 Fire_spline {
  bEnabled=false
  emitterUpdateScript [
   &EmitterState(
    Inactive Response=::NewEnumerator1
    Life Cycle Mode=::NewEnumerator1
    Loop Behavior=::NewEnumerator1
    Loop Duration Mode=::NewEnumerator1
    UseLoopDelay=false
   )
   &SpawnBurst_Instantaneous(Spawn Count=12, Use Spawn Probability=false, UseLoopCountLimit=false)
  ]
  particleSpawnScript [
   &PartitionParticles(Number Of Partitions=14)
   &InitializeParticle(
    Color=<rgba 0.7 0.7 0.7 1>
    Mesh Scale=&SelectVectorFromArray(
     Direct Array Index=$Particles.PartitionParticles.Partition
     Vector Selection Array=DataInterface() {propertyValues {FloatData [
        <1 1 1>
        <0.8 0.7 1>
        <1 1 1>
        <0.8 1 1>
        <0 0 0>
        <0.8 1 0.8>
        <0.8 1 1>
        <0 0 0>
        <1 1 1>
        <0.8 0.7 1>
        <1 1 0.7>
        <0.8 0.7 1>
        <1 1 1.2>
        <1 1 1>
       ]}}
    )
    Mesh Scale Mode=::NewEnumerator4
    UsePositionOffset=false
    Write Material Random=true
   )
   &InitialMeshOrientation(Mesh Orientation Mode=::NewEnumerator2, Rotation=<-0.25 0 0>, Use Rotation Vector=true)
   &InitialMeshOrientation(
    Mesh Orientation Mode=::NewEnumerator2
    Rotation=&SelectVectorFromArray(
     Direct Array Index=$Particles.PartitionParticles.Partition
     Vector Selection Array=DataInterface() {propertyValues {FloatData [
        <0 0 0.37>
        <0 0 0.33>
        <0.07 0 0.33>
        <0 0 0.35>
        <0 0 0.35>
        <-0.05 0 0.3>
        <0 0 0.25>
        <0 0 0.25>
        <0 0 0.33>
        <-0.07 0 0.28>
        <-0.05 0 0.37>
        <-0.1 0 0.3>
        <-0.25 0 0>
        <0 0 0.4>
       ]}}
    )
    Use Rotation Vector=true
   ) {module=InitialMeshOrientation001}
   &InitialMeshOrientation(
    Mesh Orientation Mode=::NewEnumerator2
    Rotation=&SelectVectorFromArray(
     Direct Array Index=$Particles.PartitionParticles.Partition
     Vector Selection Array=DataInterface() {propertyValues {FloatData [
        <0.2 0 0>
        <0 0 0>
        <0.25 0 0>
        <0.25 0 0>
        <0 0 0>
        <0.3 0 0>
        <0.25 0 0>
        <0.1 0 0.05>
        <0.25 0 0>
        <0.4 0 0>
        <0.25 0 0>
        <0.25 0 0>
        <0 0 0>
        <0.1 0 -0.02>
       ]}}
    )
    Use Rotation Vector=true
   ) {module=InitialMeshOrientation002}
   &CameraOffset(Camera Offset Amount=&FN_FrontEndFloatModifier(Gameplay=2, Front End=-1))
  ]
  particleUpdateScript [
   /SetVariables(Particles.Movement fade:NiagaraFloat=&Lerp_Float(A=1, B=0.05, Alpha=$System.InterpolateOverTime.InterpolateOverTime_Float.Moving Average)) {isSetParameters=true}
   /SetVariables(Particles.DanceFade:NiagaraFloat=&Lerp_Float(A=1, B=0.05, Alpha=$System.InterpolateOverTime001.InterpolateOverTime_Float.Moving Average)) {isSetParameters=true}
   /SetVariables(Particles.Alpha Scale:NiagaraFloat=&Lerp_Float(
     A=&SelectFloatFromArray(
      Direct Array Index=$Particles.PartitionParticles.Partition
      Float Selection Array=DataInterface() {propertyValues {FloatData [1,1,1,1,0,1,1,0,1,1,1,1,1,1,1]}}
     )
     B=&SelectFloatFromArray(
      Direct Array Index=$Particles.PartitionParticles.Partition
      Float Selection Array=DataInterface() {propertyValues {FloatData [1,1,1,1,0,1,1,0,1,1,1,1,1,1,1]}}
     )
     Alpha=&MakeCustomFloatFromBool(
      Bool=$User.Emote begin
      True Float=$System.InterpolateOverTime001.InterpolateOverTime_Float.Moving Average
      False Float=$System.InterpolateOverTime.InterpolateOverTime_Float.Moving Average
     )
    )) {isSetParameters=true}
   &SkeletalMeshLocation(
    Bone / Socket Index=$Particles.PartitionParticles.Partition
    Bone Sampling Mode=::NewEnumerator4
    MaskSpawnGroup=false
    Sampled Position Offset=&SelectVectorFromArray(
     Direct Array Index=$Particles.PartitionParticles.Partition
     Vector Selection Array=DataInterface() {propertyValues {FloatData [
        <15 -5 0>
        <15 3 0>
        <10 5 0>
        <10 -5 0>
        <7 12 3>
        <10 -5 -2>
        <8 5 -3>
        <10 -3 -7>
        <15 0 0>
        <17 -7 -5>
        <9 -1 2>
        <10 5 2>
        <-12 2 10>
        <12 0 0>
       ]}}
    )
    Skeletal Mesh=$System.Skeletal Mesh
   )
   &ParticleState(Kill Particles When Lifetime Has Elapsed=false, Let Infinitely Lived Particles Die When Emitter Deactivates=true)
   &Timeline(Play Rate=1.7, Timeline End=10000000)
   &ScaleMeshSize(Scale Factor=&Multiply_Vector(
     A=&SelectVectorFromArray(
      Direct Array Index=$Particles.PartitionParticles.Partition
      Vector Selection Array=DataInterface() {propertyValues {FloatData [
         <0.2 0.2 0.22>
         <0.2 0.2 0.22>
         <0.2 0.2 0.2>
         <0.2 0.2 0.2>
         <0.2 0.2 0.2>
         <0.2 0.2 0.22>
         <0.2 0.2 0.2>
         <0.2 0.2 0.22>
         <0.2 0.2 0.25>
         <0.2 0.2 0.25>
         <0.2 0.2 0.2>
         <0.2 0.2 0.2>
         <0.2 0.2 0.15>
         <0.2 0.2 0.13>
        ]}}
     )
     B=&Lerp_Vector(
      A=&Lerp_Vector(A=<1 1 1>, B=<1 1 1>, Alpha=$System.InterpolateOverTime001.InterpolateOverTime_Float.Moving Average)
      B=<0.8 0.8 0.8>
      Alpha=$System.InterpolateOverTime.InterpolateOverTime_Float.Moving Average
     )
    ))
   &ScaleMeshSize(Scale Factor=&FN_FrontEndVector3Modifier(Gameplay=<0.8 0.8 1>, Front End=<0.8 0.8 1>)) {module=ScaleMeshSize001}
   &ScaleColor(
    Scale Alpha=&Multiply_Float(
     A=&Multiply_Float(
      A=&Multiply_Float(
       A=2
       B=&FN_LOD_Distance_Remap(Start Distance=1400, End Distance=1700, Minimum Output=1, Maximum Output=0, Clamp Output=true)
      )
      B=$Particles.Alpha Scale
     )
     B=&Multiply_Float(A=$Particles.Movement fade, B=$Particles.DanceFade)
    )
    Scale RGB=&Multiply_VectorByFloat(
     Vector=<1 1 1>
     Float=&FN_FrontEndFloatModifier(
      Gameplay=&Multiply_Float(
       A=$System.Default Color TOD fix
       B=&FNTimeOfDayScalar(Night Scale=1, Morning Scale=1, Evening Scale=1)
      )
      Front End=0.4
     )
    )
    ScaleA=true
    ScaleRGB=true
   )
   &InterpolateOverTime(
    Rate Of Change:NiagaraFloat=$System.Interpolate Rate
    Target Vector Value:NiagaraPosition=$Particles.Position
    Variable Type=::NewEnumerator2
   )
   /SetVariables(Particles.Interpolate DY:Vector4f=&MakeFloat4FromVectorAndFloat("Vector (RGB)"=$Particles.InterpolateOverTime.InterpolateOverTime_V3_.Moving Average, "Float (Alpha)"=$System.Interpolate Internsity)) {isSetParameters=true}
   /SetVariables(Particles.Color1:LinearColor=&MakeLinearColorFromVectorAndFloat(
     "Vector (RGB)"=&MakeVectorFromLinearColorRGB(Color=&MultiplyLinearColor(
       Linear Color=<rgba 1.2 1.2 1.2 1>
       Scale Factor=&MakeVector4_FromFloat(Float=&FN_FrontEndFloatModifier(
         Gameplay=&Multiply_Float(
          A=&Multiply_Float(
           A=$System.Color1 TOD fix
           B=&FNTimeOfDayScalar(Night Scale=0.6, Morning Scale=1, Evening Scale=1)
          )
          B=&FN_PickFloatDependOnPlatform(
           TargetPlatformActive=1
           TargetPlatformNoActive=&Multiply_Float(A=0.8, B=&FNTimeOfDayScalar(Night Scale=1.3, Morning Scale=1, Evening Scale=1))
          )
         )
         Front End=0.9
        ))
      ))
     "Float (Alpha)"=&Multiply_Float(
      A=&Multiply_Float(
       A=&Multiply_Float(
        A=2
        B=&FN_LOD_Distance_Remap(Start Distance=1400, End Distance=1700, Minimum Output=1, Maximum Output=0, Clamp Output=true)
       )
       B=$Particles.Alpha Scale
      )
      B=&Multiply_Float(A=$Particles.Movement fade, B=$Particles.DanceFade)
     )
    )) {isSetParameters=true}
   /SetVariables(Particles.Color2=&MakeLinearColorFromVectorAndFloat(
     "Vector (RGB)"=&MakeVectorFromLinearColorRGB(Color=&MultiplyLinearColor(
       Linear Color=&FNTimeOfDayColor(Night Color=<rgba 0.5068 0.3773 1 1>, Morning Color=<rgba 1 1 1 1>, Day Color=<rgba 1 1 1 1>, Evening Color=<rgba 1 1 1 1>)
       Scale Factor=&MakeVector4_FromFloat(Float=&FN_FrontEndFloatModifier(
         Gameplay=&Multiply_Float(
          A=$System.Color2 TOD fix
          B=&FNTimeOfDayScalar(Night Scale=1, Morning Scale=1, Evening Scale=1.3)
         )
         Front End=0.7
        ))
      ))
     "Float (Alpha)"=&Multiply_Float(
      A=&Multiply_Float(
       A=&Multiply_Float(
        A=0.35
        B=&FN_LOD_Distance_Remap(Start Distance=1400, End Distance=1700, Minimum Output=1, Maximum Output=0, Clamp Output=true)
       )
       B=$Particles.Alpha Scale
      )
      B=&Multiply_Float(A=$Particles.Movement fade, B=$Particles.DanceFade)
     )
    )) {isSetParameters=true}
   /SetVariables(Particles.Color3:LinearColor=&MakeLinearColorFromVectorAndFloat(
     "Vector (RGB)"=&MakeVectorFromLinearColorRGB(Color=<rgba 0.5 0.5 0.5 1>)
     "Float (Alpha)"=&Multiply_Float(
      A=&Multiply_Float(
       A=&Multiply_Float(
        A=0.2
        B=&FN_LOD_Distance_Remap(Start Distance=1400, End Distance=1700, Minimum Output=1, Maximum Output=0, Clamp Output=true)
       )
       B=$Particles.Alpha Scale
      )
      B=&Multiply_Float(A=$Particles.Movement fade, B=$Particles.DanceFade)
     )
    )) {isSetParameters=true}
   &DynamicMaterialParameters(
    Index 0 Param 1=&SelectFloatFromArray(
     Direct Array Index=$Particles.PartitionParticles.Partition
     Float Selection Array=DataInterface() {propertyValues {FloatData [0.6,0.6,0.6,0.6,0,0.6,0.6,0,0.55,0.55,0.6,0.6,0.7,0.9]}}
    )
    Index 0 Param 2=$Particles.Timeline.Time
    Index 0 Param 3=&FN_FrontEndFloatModifier(Gameplay=10, Front End=5)
    Index 0 Param 4=&SelectFloatFromArray(
     Direct Array Index=$Particles.PartitionParticles.Partition
     Float Selection Array=DataInterface() {propertyValues {FloatData [0.1,0.5,0.1,0.5,0.25,0.1,0.5,0.25,0.1,0.5,0.1,0.5,0.1,0.5]}}
    )
    Index 2 Param 1=0.7
    Module.Index0Param0Write=true
    Module.Index0Param1Write=true
    Module.Index0Param2Write=true
    Module.Index0Param3Write=true
    Module.Index2Param0Write=true
    Module.Index2Param1Write=true
    Module.Index2Param2Write=true
    Module.Index2Param3Write=true
    Write Parameter Index 2=true
   )
  ]
  rendererClasses [&NiagaraMeshRendererProperties]
  renderers [
   {rendererClass=&NiagaraMeshRendererProperties, rendererIndex=0}
   {rendererClass=&NiagaraMeshRendererProperties, rendererIndex=1}
   {rendererClass=&NiagaraMeshRendererProperties, rendererIndex=2}
   {rendererClass=&NiagaraMeshRendererProperties, rendererIndex=3}
  ]
 }
 ParticleLight {
  emitterUpdateScript [
   &EmitterState(
    Inactive Response=::NewEnumerator1
    Life Cycle Mode=::NewEnumerator1
    Loop Behavior=::NewEnumerator1
    Loop Delay=1.4
    Loop Duration Mode=::NewEnumerator1
    UseLoopDelay=false
   )
   &SpawnBurst_Instantaneous(Spawn Count=2, Use Spawn Probability=false, UseLoopCountLimit=false)
  ]
  particleSpawnScript [
   &PartitionParticles(Number Of Partitions=5)
   &InitializeParticle(
    Color=<rgba 2 0 0 0.7>
    Sprite Rotation Mode=::NewEnumerator2
    Sprite Size Mode=::NewEnumerator3
    Uniform Sprite Size=80
    UsePositionOffset=false
    Write Material Random=true
   ) {module=InitializeParticle001}
  ]
  particleUpdateScript [
   &SkeletalMeshLocation(
    Bone / Socket Index=$Particles.PartitionParticles.Partition
    Bone Sampling Mode=::NewEnumerator4
    MaskSpawnGroup=false
    Sampled Position Offset Coordinate Space:ENiagaraSkelMeshTransforms=::NewEnumerator1
    Skeletal Mesh=DataInterface() {propertyValues {
      DefaultMesh=null
      ExcludeBoneName=root
      OptionalComponentClass=null
      PreviewMesh=/BRCosmetics/Characters/Player/Female/Medium/Bodies/F_MED_MadameMoth_Posh/Meshes/F_MED_MadameMoth_Posh
      SkinningMode=SkinOnTheFly
      SoftSourceActor=null
      SourceMode=Default
      UvSetIndex=0
      WholeMeshLOD=-1
      bExcludeBone=false
      bReadDeformedGeometry=true
      bRequireCurrentFrameData=true
      ComponentTags []
      FilteredBones [calf_l,calf_r]
      FilteredSockets []
      MeshUserParameter {parameter {name=None, typeDefHandle {registeredTypeIndex=1}, varData []}}
      SamplingRegions []
     }}
   ) {module=SkeletalMeshLocation001}
   &ParticleState(Kill Particles When Lifetime Has Elapsed=false, Let Infinitely Lived Particles Die When Emitter Deactivates=true)
   &ScaleSpriteSize(Scale Factor=&Vector2DFromFloat(Value=&FloatFromCurve(FloatCurve=DataInterface() {propertyValues {Curve {keys [
          {interpMode=RCIM_Cubic, leaveTangent=5.3269, tangentMode=RCTM_Break, time=0, value=0}
          {arriveTangent=0.6, interpMode=RCIM_Cubic, leaveTangent=0.6, time=0.2, value=0.54}
          {arriveTangent=0.0611, interpMode=RCIM_Cubic, tangentMode=RCTM_Break, time=1, value=0.6}
         ]}}})))
   &FN_Light_Attributes(
    Color=<rgba 10 0.5208 0.5208 1>
    Create Color Attribute=true
    Create Exponent Attribute=true
    Create Radius Attribute=true
    Exponent=2
    Radius=70
    Scale Color=&FNTimeOfDayScalar(Night Scale=0.5, Morning Scale=1, Evening Scale=1)
   )
  ]
  rendererClasses [&NiagaraLightRendererProperties]
  renderers [{rendererClass=&NiagaraLightRendererProperties, rendererIndex=0}]
 }
 Smoke {
  emitterUpdateScript [
   &EmitterState(Inactive Response=::NewEnumerator1, Life Cycle Mode=::NewEnumerator1, Loop Delay=2.9, Loop Duration=1, UseLoopDelay=false)
   &SpawnPerUnit(
    Max Movement Threshold=1000
    Module.Use Max Movement Threshold=true
    Module.Use Movement Tolerance=false
    Spawn Spacing:NiagaraFloat=&Multiply_Float(
     A=25
     B=&MakeCustomFloatFromBool(
      Bool=&SetBoolByFloatComparison(A=&VectorLength(Vector=$Engine.Owner.Velocity), B=100)
      True Float=1
      False Float=0
     )
    )
    Use Spawn Probability=false
   )
  ]
  particleSpawnScript [
   &PartitionParticles()
   /SetVariables(Particles.Random size:NiagaraFloat=&RandomRangeFloat(Minimum=0.2, Maximum=0.25)) {isSetParameters=true, module=SetVariables_#8}
   &InitializeParticle(
    Color=&MakeCustomLinearColorFromBool(Bool=&RandomBool(), True Color=<rgba 1 0.0625 0.1094 1>, False Color=<rgba 1 0.0625 0.1094 1>)
    Lifetime Max=0.6
    Lifetime Min=0.4
    Lifetime Mode=::NewEnumerator1
    Sprite Rotation Mode=::NewEnumerator2
    Sprite Size Mode=::NewEnumerator1
    Sprite UV Mode=::NewEnumerator3
    Uniform Sprite Size Max=20
    Uniform Sprite Size Min=10
    UsePositionOffset=false
    Write Material Random=true
   ) {module=InitializeParticle001}
   &SkeletalMeshLocation(
    Bone / Socket Index=$Particles.PartitionParticles.Partition
    Bone Sampling Mode=::NewEnumerator4
    MaskSpawnGroup=false
    Sampled Position Offset=&AddVectorToPosition(
     Position=<0 0 0>
     Vector=&SelectVectorFromArray(
      Direct Array Index=$Particles.PartitionParticles.Partition
      Vector Selection Array=DataInterface() {propertyValues {FloatData [<0 0 0>, <0 0 0>]}}
     )
    )
    Skeletal Mesh=$System.Skeletal Mesh two side
   )
   &ShapeLocation(MaskSpawnGroup=false, Override Seed=false, Sphere Radius=3, UseOwnerScale=true)
   &CameraOffset(Camera Offset Amount=1)
  ]
  particleUpdateScript [
   /SetVariables(Particles.SubImageIndex=&FloatFromCurve(FloatCurve=DataInterface() {propertyValues {Curve {keys [<at 0 7>,<at 1 15>]}}})) {isSetParameters=true}
   &CurlNoiseForce(Noise Frequency=100, Noise Strength=&Multiply_Float(A=100, B=1), Pan Noise=true)
   &ParticleState()
   &ScaleSpriteSize(Uniform Curve Sprite Scale=DataInterface() {propertyValues {Curve {keys [{interpMode=RCIM_Cubic, tangentMode=RCTM_User, time=0, value=2.5}, {interpMode=RCIM_Cubic, tangentMode=RCTM_User, time=1, value=0}]}}}) {module=ScaleSpriteSize001}
   &SpriteRotationRate(Rotation Rate=&RandomRangeFloat(Minimum=-5, Maximum=5)) {module=SpriteRotationRate001}
   &Drag(Use Linear Drag=true, Use Rotational Drag=true)
   &InheritSourceMovement(Applied Position Scale=<0.7 0.7 0.7>, ApplyForce=false, ApplyPosition=true, ApplyPositionFalloff=false)
   &SolveForcesAndVelocity(Clamp Velocity=false, Limit Acceleration=false)
   &ScaleColor(
    Scale Alpha=&Multiply_Float(
     A=&FloatFromCurve(
      FloatCurve=DataInterface() {propertyValues {Curve {keys [
          {interpMode=RCIM_Cubic, tangentMode=RCTM_User, time=0, value=0}
          <at 0.5 0.8>
          {interpMode=RCIM_Cubic, tangentMode=RCTM_User, time=1, value=0}
         ]}}}
      Scale Curve=&FN_FrontEndFloatModifier(Gameplay=0.5, Front End=0.2)
     )
     B=&FN_LOD_Distance_Remap(Start Distance=1400, End Distance=1700, Minimum Output=1, Maximum Output=0, Clamp Output=true)
    )
    Scale RGB=&Multiply_VectorByFloat(
     Vector=&MakeVectorFromLinearColorRGB(Color=&ColorFromCurve(ColorCurve=DataInterface() {propertyValues {
         AlphaCurve {keys []}
         BlueCurve {keys [<at 0 0.5506 RCIM_Cubic>,<at 1 0 RCIM_Cubic>]}
         GreenCurve {keys [<at 0 0.1899 RCIM_Cubic>,<at 1 0 RCIM_Cubic>]}
         RedCurve {keys [<at 0 2 RCIM_Cubic>,<at 1 2 RCIM_Cubic>]}
        }}))
     Float=&FN_FrontEndFloatModifier(Gameplay=1, Front End=0.5)
    )
    ScaleA=true
    ScaleRGB=true
   ) {module=ScaleColor001}
  ]
  rendererClasses [&NiagaraSpriteRendererProperties]
  renderers [{rendererClass=&NiagaraSpriteRendererProperties}]
 }
}
summary {
 systemName=NS_NPC_MadameMoth_Skin_Posh
 emitters [
  {emitterName=Fire_pieces_Idle, rendererClasses [&NiagaraSpriteRendererProperties]}
  {emitterName=Smoke, rendererClasses [&NiagaraSpriteRendererProperties]}
  {bEnabled=false, emitterName=Fire_spline, rendererClasses [&NiagaraMeshRendererProperties]}
  {emitterName=Fire_pieces_Idle001, rendererClasses [&NiagaraSpriteRendererProperties]}
  {emitterName=ParticleLight, rendererClasses [&NiagaraLightRendererProperties]}
  {emitterName=Fire_pieces_Idle003, rendererClasses [&NiagaraSpriteRendererProperties]}
 ]
 userVariables [
  User.bIsFrontEnd {type=NiagaraBool}
  User.bIsFrontEndPreview {type=NiagaraBool}
  User.Emote begin {type=NiagaraBool}
  User.IsSkydiving {type=NiagaraBool}
 ]
}
