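# Physics asset for the Madame Moth Posh NPC body mesh, set up for the dynamic bones of the character's coat.
# Five four-bone capsule chains (left/right back and side, right front only) hang from the pelvis with limited swing and locked twist, plus a free-jointed frill off the left side chain.
# Thigh and calf capsules provide leg shapes on free joints, and an oversized root capsule sunk below the origin appears to act as a floor.
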PhysicsAsset /BRCosmetics/Characters/Player/Female/Medium/Bodies/F_MED_MadameMoth_Posh_NPC/Meshes/F_MED_MadameMoth_Posh_NPC_Physics
bodies {
 calf_l [{shapeType=Capsule, length=40, radius=7, shapeName=calf_l_capsule, center=<-29.0341 0 0>, rotation=<rot 90 0 0>}]
 calf_r [{shapeType=Capsule, length=40, radius=7, shapeName=calf_r_capsule, center=<29.2262 0 0>, rotation=<rot 90 0 0>}]
 dyn_L_backCoat_1 [{shapeType=Capsule, length=9.3062, radius=2.0511, shapeName=dyn_coat_bk_l_01_capsule, center=<7.4311 0 0>, rotation=<rot 90 0 0>}]
 dyn_L_backCoat_2 [{shapeType=Capsule, length=8.9445, radius=3, shapeName=dyn_coat_bk_l_02_capsule, center=<7.4311 0 0>, rotation=<rot 90 0 0>}]
 dyn_L_backCoat_3 [{shapeType=Capsule, length=10, radius=5, shapeName=dyn_coat_bk_l_03_capsule, center=<9.2473 0 0>, rotation=<rot 90 0 90>}]
 dyn_L_backCoat_4 [{shapeType=Capsule, length=10, radius=6, shapeName=dyn_coat_bk_l_04_capsule, center=<4.9151 -1.7715 0>, rotation=<rot 90 180 -90>}]
 dyn_L_sideCoat_1 [{shapeType=Capsule, length=9.3062, radius=4, shapeName=dyn_coat_bk_l_01_capsule, center=<8.8203 0 0>, rotation=<rot 90 0 0>}]
 dyn_L_sideCoat_2 [{shapeType=Capsule, length=8.9445, radius=5, shapeName=dyn_coat_bk_l_02_capsule, center=<7.4311 0 0>, rotation=<rot 90 0 90>}]
 dyn_L_sideCoat_3 [{shapeType=Capsule, length=10, radius=5, shapeName=dyn_coat_bk_l_03_capsule, center=<9.2473 0 0>, rotation=<rot 90 0 90>}]
 dyn_L_sideCoat_4 [{shapeType=Capsule, length=12, radius=6, shapeName=dyn_coat_bk_l_04_capsule, center=<8.3844 0 0>, rotation=<rot 90 153.4349 -116.5651>}]
 dyn_L_sideCoat_frill [{shapeType=Capsule, length=12, radius=6, shapeName=dyn_coat_bk_l_04_capsule, center=<7.9944 0 0>, rotation=<rot 90 180 -90>}]
 dyn_R_backCoat_1 [{shapeType=Capsule, length=9.3062, radius=2.0511, shapeName=dyn_coat_bk_l_01_capsule, center=<7.4311 0 0>, rotation=<rot 90 0 0>}]
 dyn_R_backCoat_2 [{shapeType=Capsule, length=8.9445, radius=2.4128, shapeName=dyn_coat_bk_l_02_capsule, center=<7.4311 0 0>, rotation=<rot 90 0 0>}]
 dyn_R_backCoat_3 [{shapeType=Capsule, length=10, radius=5, shapeName=dyn_coat_bk_l_03_capsule, center=<9.2473 0 0>, rotation=<rot 90 0 90>}]
 dyn_R_backCoat_4 [{shapeType=Capsule, length=25, radius=6, shapeName=dyn_coat_bk_l_04_capsule, center=<5.3753 0 0>, rotation=<rot 90 9.1265 93.6146>}]
 dyn_R_frontCoat_1 [{shapeType=Capsule, length=10, radius=3, shapeName=dyn_coat_bk_l_01_capsule, center=<10 0 1.78>, rotation=<rot 87.7165 0.0001 0.0001>}]
 dyn_R_frontCoat_2 [{
   shapeType=Capsule
   length=9
   radius=5
   shapeName=dyn_coat_bk_l_02_capsule
   center=<10.2268 0.0885 3.7178>
   rotation=<rot 81.4876 -56.119 34.1754>
  }]
 dyn_R_frontCoat_3 [{
   shapeType=Capsule
   length=12.8868
   radius=5
   shapeName=dyn_coat_bk_l_03_capsule
   center=<10.1383 -0.0101 -0.0175>
   rotation=<rot 70.4427 -96.1188 -3.3078>
  }]
 dyn_R_frontCoat_4 [{shapeType=Capsule, length=20, radius=6, shapeName=dyn_coat_bk_l_04_capsule, center=<8.1732 0 0>, rotation=<rot 90 169.0241 -98.4271>}]
 dyn_R_sideCoat_1 [{shapeType=Capsule, length=12, radius=3, shapeName=dyn_coat_bk_l_01_capsule, center=<9 0 6>, rotation=<rot 87.7165 0.0001 0.0001>}]
 dyn_R_sideCoat_2 [{shapeType=Capsule, length=8.9445, radius=5, shapeName=dyn_coat_bk_l_02_capsule, center=<8.6359 0 4.8908>, rotation=<rot 85.9428 0.0001 90>}]
 dyn_R_sideCoat_3 [{shapeType=Capsule, length=12.8868, radius=5, shapeName=dyn_coat_bk_l_03_capsule, center=<7.6423 0 4.2398>, rotation=<rot 90 180 -90>}]
 dyn_R_sideCoat_4 [{shapeType=Capsule, length=25, radius=6, shapeName=dyn_coat_bk_l_04_capsule, center=<7.4311 0 2.0703>, rotation=<rot 90 153.4349 -116.5651>}]
 pelvis [{shapeType=Capsule, length=1, radius=1}]
 root [{shapeType=Capsule, length=0.1, radius=250, shapeName=root_capsule, center=<0 0 -248>}]
 thigh_l [{shapeType=Capsule, length=30, radius=7.5, shapeName=thigh_l_capsule, center=<-20.4551 0 0>, rotation=<rot 85.1603 179.9997 179.9999>}]
 thigh_r [{shapeType=Capsule, length=30, radius=7, shapeName=thigh_r_capsule, center=<21.1903 0 0>, rotation=<rot 90 0 0>}]
}
constraints [
 {
  bone1Name=pelvis
  bone2Name=root
  swing1LimitDegrees=45
  swing1Motion=Free
  swing2LimitDegrees=45
  swing2Motion=Free
  twistLimitDegrees=45
  twistMotion=Free
 }
 {
  bone1Name=dyn_L_backCoat_2
  bone2Name=dyn_L_backCoat_1
  swing1LimitDegrees=20
  swing2LimitDegrees=20
  twistLimitDegrees=1
  twistMotion=Locked
 }
 {
  bone1Name=dyn_L_backCoat_3
  bone2Name=dyn_L_backCoat_2
  swing1LimitDegrees=20
  swing2LimitDegrees=20
  twistLimitDegrees=1
  twistMotion=Locked
 }
 {
  bone1Name=dyn_L_backCoat_4
  bone2Name=dyn_L_backCoat_3
  swing1LimitDegrees=20
  swing2LimitDegrees=20
  twistLimitDegrees=1
  twistMotion=Locked
 }
 {
  bone1Name=dyn_L_sideCoat_2
  bone2Name=dyn_L_sideCoat_1
  swing1LimitDegrees=30
  swing2LimitDegrees=20
  twistLimitDegrees=1
  twistMotion=Locked
 }
 {
  bone1Name=dyn_L_sideCoat_3
  bone2Name=dyn_L_sideCoat_2
  swing1LimitDegrees=30
  swing2LimitDegrees=20
  twistLimitDegrees=1
  twistMotion=Locked
 }
 {
  bone1Name=dyn_L_sideCoat_4
  bone2Name=dyn_L_sideCoat_3
  swing1LimitDegrees=20
  swing2LimitDegrees=20
  twistLimitDegrees=1
  twistMotion=Locked
 }
 {
  bone1Name=dyn_R_backCoat_2
  bone2Name=dyn_R_backCoat_1
  swing1LimitDegrees=20
  swing2LimitDegrees=20
  twistLimitDegrees=1
  twistMotion=Locked
 }
 {
  bone1Name=dyn_R_backCoat_3
  bone2Name=dyn_R_backCoat_2
  swing1LimitDegrees=20
  swing2LimitDegrees=20
  twistLimitDegrees=1
  twistMotion=Locked
 }
 {
  bone1Name=dyn_R_backCoat_4
  bone2Name=dyn_R_backCoat_3
  swing1LimitDegrees=20
  swing2LimitDegrees=20
  twistLimitDegrees=1
  twistMotion=Locked
 }
 {
  bone1Name=dyn_R_frontCoat_2
  bone2Name=dyn_R_frontCoat_1
  swing1LimitDegrees=20
  swing2LimitDegrees=20
  twistLimitDegrees=1
  twistMotion=Locked
 }
 {
  bone1Name=dyn_R_frontCoat_3
  bone2Name=dyn_R_frontCoat_2
  swing1LimitDegrees=20
  swing2LimitDegrees=20
  twistLimitDegrees=1
  twistMotion=Locked
 }
 {
  bone1Name=dyn_R_frontCoat_4
  bone2Name=dyn_R_frontCoat_3
  swing1LimitDegrees=20
  swing2LimitDegrees=20
  twistLimitDegrees=1
  twistMotion=Locked
 }
 {
  bone1Name=dyn_R_sideCoat_2
  bone2Name=dyn_R_sideCoat_1
  swing1LimitDegrees=20
  swing2LimitDegrees=20
  twistLimitDegrees=1
  twistMotion=Locked
 }
 {
  bone1Name=dyn_R_sideCoat_3
  bone2Name=dyn_R_sideCoat_2
  swing1LimitDegrees=20
  swing2LimitDegrees=20
  twistLimitDegrees=1
  twistMotion=Locked
 }
 {
  bone1Name=dyn_R_sideCoat_4
  bone2Name=dyn_R_sideCoat_3
  swing1LimitDegrees=20
  swing2LimitDegrees=20
  twistLimitDegrees=1
  twistMotion=Locked
 }
 {
  bone1Name=thigh_l
  bone2Name=pelvis
  swing1LimitDegrees=45
  swing1Motion=Free
  swing2LimitDegrees=45
  swing2Motion=Free
  twistLimitDegrees=45
  twistMotion=Free
 }
 {
  bone1Name=calf_l
  bone2Name=thigh_l
  swing1LimitDegrees=45
  swing1Motion=Free
  swing2LimitDegrees=45
  swing2Motion=Free
  twistLimitDegrees=45
  twistMotion=Free
 }
 {
  bone1Name=thigh_r
  bone2Name=pelvis
  swing1LimitDegrees=45
  swing1Motion=Free
  swing2LimitDegrees=45
  swing2Motion=Free
  twistLimitDegrees=45
  twistMotion=Free
 }
 {
  bone1Name=calf_r
  bone2Name=thigh_r
  swing1LimitDegrees=45
  swing1Motion=Free
  swing2LimitDegrees=45
  swing2Motion=Free
  twistLimitDegrees=45
  twistMotion=Free
 }
 {bone1Name=dyn_L_backCoat_1, bone2Name=pelvis, swing1LimitDegrees=20, swing2LimitDegrees=20, twistLimitDegrees=0, twistMotion=Locked}
 {bone1Name=dyn_L_sideCoat_1, bone2Name=pelvis, swing1LimitDegrees=30, swing2LimitDegrees=30, twistLimitDegrees=0, twistMotion=Locked}
 {bone1Name=dyn_R_backCoat_1, bone2Name=pelvis, swing1LimitDegrees=30, swing2LimitDegrees=30, twistLimitDegrees=0, twistMotion=Locked}
 {bone1Name=dyn_R_frontCoat_1, bone2Name=pelvis, swing1LimitDegrees=25, swing2LimitDegrees=25, twistLimitDegrees=0, twistMotion=Locked}
 {bone1Name=dyn_R_sideCoat_1, bone2Name=pelvis, swing1LimitDegrees=25, swing2LimitDegrees=25, twistLimitDegrees=0, twistMotion=Locked}
 {
  bone1Name=dyn_L_sideCoat_frill
  bone2Name=dyn_L_sideCoat_3
  swing1LimitDegrees=45
  swing1Motion=Free
  swing2LimitDegrees=45
  swing2Motion=Free
  twistLimitDegrees=45
  twistMotion=Free
 }
]
