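# Post-style cosmetic AnimBP for the Madame Moth Posh NPC body, copying the attached parent mesh's pose and adding secondary motion.
# Chained AnimDynamics nodes sway bangs, buns, fur collar and holster, then constraints tie right coat bones to the thigh.
# A pelvis-space RigidBody node takes a gravity override, updated each frame by a deprecated function, as external force.
# Additive coat bone offsets around it switch instantly between standing and crouching via crouch events.
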AnimBlueprint /BRCosmetics/Characters/Player/Female/Medium/Bodies/F_MED_MadameMoth_Posh_NPC/Meshes/F_MED_MadameMoth_Posh_NPC_AnimBP
target_skeleton=&/F_MED_MadameMoth_Posh_NPC_Skeleton
graphs [
 AnimGraph {nodes [
   n1 = Anim Dynamics - Bone: dyn_bangBase_1(ComponentPose=Local To Component(LocalPose=Copy Pose From Mesh() {bUseAttachedParent=true}), Alpha=1, ExternalForce="0,0,0") {
    angularDampingOverride=1
    angularSpringConstant=25000
    bAngularSpring=true
    bOverrideAngularDamping=true
    bUseGravityOverride=true
    gravityScale=0.35
    simulationSpace=RootRelative
    alphaBoolBlend {blendOption=Linear}
    boundBone {boneName=dyn_bangBase_1}
    componentLinearAccScale=<1 1 1>
    componentLinearVelScale=<1 1 1>
    gravityOverride {z=-980}
    physicsBodyDefinitions [{
      boundBone {boneName=dyn_bangBase_1}
      boxExtents=<41 41 41>
      constraintSetup {angularLimitsMax=<0 15 0.5>, angularLimitsMin=<0 -15 0>, angularTarget=<1 0 0>}
      localJointOffset=<1 0 0>
     }]
   }
   n2 = Anim Dynamics - Bone: dyn_bangBase_2(ComponentPose=@n1, Alpha=1, ExternalForce="0,0,0") {
    angularDampingOverride=1
    angularSpringConstant=25000
    bAngularSpring=true
    bOverrideAngularDamping=true
    bUseGravityOverride=true
    gravityScale=0.35
    simulationSpace=RootRelative
    alphaBoolBlend {blendOption=Linear}
    boundBone {boneName=dyn_bangBase_2}
    componentLinearAccScale=<1 1 1>
    componentLinearVelScale=<1 1 1>
    gravityOverride {z=-980}
    physicsBodyDefinitions [{
      boundBone {boneName=dyn_bangBase_2}
      boxExtents=<41 41 41>
      constraintSetup {angularLimitsMax=<0 2 0.5>, angularLimitsMin=<0 -2 0>, angularTarget=<1 0 0>}
      localJointOffset=<1 0 0>
     }]
   }
   n3 = Anim Dynamics - Bone: dyn_bangBase_3(ComponentPose=@n2, Alpha=1, ExternalForce="0,0,0") {
    angularDampingOverride=1
    angularSpringConstant=25000
    bAngularSpring=true
    bOverrideAngularDamping=true
    bUseGravityOverride=true
    gravityScale=0.35
    simulationSpace=RootRelative
    alphaBoolBlend {blendOption=Linear}
    boundBone {boneName=dyn_bangBase_3}
    componentLinearAccScale=<1 1 1>
    componentLinearVelScale=<1 1 1>
    gravityOverride {z=-980}
    physicsBodyDefinitions [{
      boundBone {boneName=dyn_bangBase_3}
      boxExtents=<41 41 41>
      constraintSetup {angularLimitsMax=<0 5 1>, angularLimitsMin=<0 -2 0>, angularTarget=<1 0 0>}
      localJointOffset=<2 0 0>
     }]
   }
   n4 = Anim Dynamics - Bone: dyn_bangFringe_1(ComponentPose=@n3, Alpha=1, ExternalForce="0,0,0") {
    angularDampingOverride=1
    angularSpringConstant=25000
    bAngularSpring=true
    bOverrideAngularDamping=true
    bUseGravityOverride=true
    gravityScale=0.35
    simulationSpace=RootRelative
    alphaBoolBlend {blendOption=Linear}
    boundBone {boneName=dyn_bangFringe_1}
    componentLinearAccScale=<1 1 1>
    componentLinearVelScale=<1 1 1>
    gravityOverride {z=-980}
    physicsBodyDefinitions [{
      boundBone {boneName=dyn_bangFringe_1}
      boxExtents=<40 40 40>
      constraintSetup {angularLimitsMax=<1 1 1>, angularLimitsMin=<0 -1 -1>, angularTarget=<1 0 0>}
      localJointOffset=<2 0 0>
     }]
   }
   n5 = Anim Dynamics - Bone: dyn_bangFringe_3(ComponentPose=@n4, Alpha=1, ExternalForce="0,0,0") {
    angularDampingOverride=1
    angularSpringConstant=25000
    bAngularSpring=true
    bOverrideAngularDamping=true
    bUseGravityOverride=true
    gravityScale=0.35
    simulationSpace=RootRelative
    alphaBoolBlend {blendOption=Linear}
    boundBone {boneName=dyn_bangFringe_3}
    componentLinearAccScale=<1 1 1>
    componentLinearVelScale=<1 1 1>
    gravityOverride {z=-980}
    physicsBodyDefinitions [{
      boundBone {boneName=dyn_bangFringe_3}
      boxExtents=<40 40 40>
      constraintSetup {angularLimitsMax=<0 3 1>, angularLimitsMin=<0 -3 -1>, angularTarget=<1 0 0>}
      localJointOffset=<2 0 0>
     }]
   }
   n6 = Anim Dynamics - Bone: dyn_R_bun1(ComponentPose=@n5, Alpha=1, ExternalForce="0,0,0") {
    angularDampingOverride=1
    angularSpringConstant=20000
    bAngularSpring=true
    bOverrideAngularDamping=true
    bUseGravityOverride=true
    gravityScale=0.35
    alphaBoolBlend {blendOption=Linear}
    boundBone {boneName=dyn_R_bun1}
    componentLinearAccScale=<1 1 1>
    componentLinearVelScale=<1 1 1>
    gravityOverride {z=-980}
    physicsBodyDefinitions [{
      boundBone {boneName=dyn_R_bun1}
      boxExtents=<30 30 30>
      constraintSetup {angularLimitsMax=<0 8 8>, angularLimitsMin=<0 -1 -1>, angularTarget=<1 0 0>}
      localJointOffset=<2 0 0>
     }]
   }
   n7 = Anim Dynamics - Bone: dyn_L_bun1(ComponentPose=@n6, Alpha=1, ExternalForce="0,0,0") {
    angularDampingOverride=1
    angularSpringConstant=20000
    bAngularSpring=true
    bOverrideAngularDamping=true
    bUseGravityOverride=true
    gravityScale=0.35
    alphaBoolBlend {blendOption=Linear}
    boundBone {boneName=dyn_L_bun1}
    componentLinearAccScale=<0.5 0.5 0.5>
    componentLinearVelScale=<0.5 0.5 0.5>
    gravityOverride {z=-980}
    physicsBodyDefinitions [{
      boundBone {boneName=dyn_L_bun1}
      boxExtents=<30 30 30>
      constraintSetup {angularLimitsMax=<0 8 8>, angularLimitsMin=<0 -1 -1>, angularTarget=<1 0 0>}
      localJointOffset=<2 0 0>
     }]
   }
   n8 = Anim Dynamics - Bone: dyn_C_bun1(ComponentPose=@n7, Alpha=1, ExternalForce="0,0,0") {
    angularDampingOverride=1
    angularSpringConstant=20000
    bAngularSpring=true
    bOverrideAngularDamping=true
    bUseGravityOverride=true
    gravityScale=0.35
    alphaBoolBlend {blendOption=Linear}
    boundBone {boneName=dyn_C_bun1}
    componentLinearAccScale=<1 1 1>
    componentLinearVelScale=<1 1 1>
    gravityOverride {z=-980}
    physicsBodyDefinitions [{
      boundBone {boneName=dyn_C_bun1}
      boxExtents=<30 30 30>
      constraintSetup {angularLimitsMax=<0 8 8>, angularLimitsMin=<0 -1 -1>, angularTarget=<1 0 0>}
      localJointOffset=<2 0 0>
     }]
   }
   n9 = Anim Dynamics - Bone: dyn_L_furCollar_adjust_1(ComponentPose=@n8, Alpha=1, ExternalForce="0,0,0") {
    angularDampingOverride=1
    angularSpringConstant=20000
    bAngularSpring=true
    bOverrideAngularDamping=true
    bUseGravityOverride=true
    gravityScale=0.35
    simulationSpace=RootRelative
    alphaBoolBlend {blendOption=Linear}
    boundBone {boneName=dyn_L_furCollar_adjust_1}
    componentLinearAccScale=<0.3 0.3 0.3>
    componentLinearVelScale=<0.3 0.3 0.3>
    gravityOverride {z=-980}
    physicsBodyDefinitions [{
      boundBone {boneName=dyn_L_furCollar_adjust_1}
      boxExtents=<20 20 20>
      constraintSetup {angularLimitsMax=<1 6 0.5>, angularLimitsMin=<-1 -1 0>, angularTarget=<1 0 0>}
      localJointOffset=<3 0 0>
     }]
   }
   n10 = Anim Dynamics - Bone: dyn_L_furCollar_adjust_2(ComponentPose=@n9, Alpha=1, ExternalForce="0,0,0") {
    angularDampingOverride=1
    angularSpringConstant=10000
    bAngularSpring=true
    bOverrideAngularDamping=true
    bUseGravityOverride=true
    gravityScale=0.35
    simulationSpace=RootRelative
    alphaBoolBlend {blendOption=Linear}
    boundBone {boneName=dyn_L_furCollar_adjust_2}
    componentLinearAccScale=<0.3 0.3 0.3>
    componentLinearVelScale=<0.3 0.3 0.3>
    gravityOverride {z=-980}
    physicsBodyDefinitions [{
      boundBone {boneName=dyn_L_furCollar_adjust_2}
      boxExtents=<50 50 50>
      constraintSetup {angularLimitsMax=<0.5 10 3>, angularLimitsMin=<0 -3 -3>, angularTarget=<1 0 0>}
      localJointOffset=<4 0 0>
     }]
   }
   n11 = Anim Dynamics - Bone: dyn_R_furCollar_adjust_1(ComponentPose=@n10, Alpha=1, ExternalForce="0,0,0") {
    angularDampingOverride=1
    angularSpringConstant=10000
    bAngularSpring=true
    bOverrideAngularDamping=true
    bUseGravityOverride=true
    gravityScale=0.35
    simulationSpace=RootRelative
    alphaBoolBlend {blendOption=Linear}
    boundBone {boneName=dyn_R_furCollar_adjust_1}
    componentLinearAccScale=<0.3 0.3 0.3>
    componentLinearVelScale=<0.3 0.3 0.3>
    gravityOverride {z=-980}
    physicsBodyDefinitions [{
      boundBone {boneName=dyn_R_furCollar_adjust_1}
      boxExtents=<20 20 20>
      constraintSetup {angularLimitsMax=<1 6 0.5>, angularLimitsMin=<-1 -2 0>, angularTarget=<1 0 0>}
      localJointOffset=<1 0 0>
     }]
   }
   n12 = Anim Dynamics - Bone: dyn_R_furCollar_adjust_2(ComponentPose=@n11, Alpha=1, ExternalForce="0,0,0") {
    angularDampingOverride=1
    angularSpringConstant=10000
    bOverrideAngularDamping=true
    bUseGravityOverride=true
    gravityScale=0.35
    simulationSpace=RootRelative
    alphaBoolBlend {blendOption=Linear}
    boundBone {boneName=dyn_R_furCollar_adjust_2}
    componentLinearAccScale=<0.3 0.3 0.3>
    componentLinearVelScale=<0.3 0.3 0.3>
    gravityOverride {z=-980}
    physicsBodyDefinitions [{
      boundBone {boneName=dyn_R_furCollar_adjust_2}
      boxExtents=<50 50 50>
      constraintSetup {angularLimitsMax=<0.5 10 3>, angularLimitsMin=<0 -2 -3>, angularTarget=<1 0 0>}
      localJointOffset=<3.5 0 0>
     }]
   }
   n13 = Anim Dynamics - Bone: dyn_gun_holster(ComponentPose=@n12, Alpha=1, ExternalForce="0,0,0") {
    angularDampingOverride=1
    angularSpringConstant=25000
    bAngularSpring=true
    bOverrideAngularDamping=true
    bUseGravityOverride=true
    gravityScale=0.35
    simulationSpace=BoneRelative
    alphaBoolBlend {blendOption=Linear}
    boundBone {boneName=dyn_gun_holster}
    componentLinearAccScale=<1 1 1>
    componentLinearVelScale=<1 1 1>
    gravityOverride {z=-980}
    physicsBodyDefinitions [{
      boundBone {boneName=dyn_gun_holster}
      boxExtents=<30 30 30>
      constraintSetup {angularLimitsMax=<0 4 8>, angularLimitsMin=<0 0 -8>, angularTarget=<1 0 0>}
      localJointOffset=<5 0 0>
     }]
    relativeSpaceBone {boneName=thigh_l}
   }
   n14 = "Constraint - dyn_R_frontCoat_MSTR (P)"(ComponentPose=@n13, Alpha=1, ConstraintWeights_0=1) {
    alphaBoolBlend {blendOption=Linear}
    boneToModify {boneName=dyn_R_frontCoat_MSTR}
    constraintSetup [{targetBone {boneName=thigh_r}}]
    constraintWeights [1]
   }
   n15 = "Constraint - dyn_R_sideCoat_MSTR (P)"(ComponentPose=@n14, Alpha=1, ConstraintWeights_0=1) {
    alphaBoolBlend {blendOption=Linear}
    boneToModify {boneName=dyn_R_sideCoat_MSTR}
    constraintSetup [{targetBone {boneName=thigh_r}}]
    constraintWeights [1]
   }
   n16 = Get crouching alpha() -> (crouching alpha:double)
   n17 = "Transform (Modify) Bone - Bone: dyn_L_backCoat_MSTR"(ComponentPose=@n15, Alpha=@n16, Translation="0.000,-3.390,-0.000", Rotation="-0.000,0.000,12.094", Scale="1,1,1") {
    rotationMode=BMM_Additive
    translationMode=BMM_Additive
    alphaBoolBlend {blendOption=Linear}
    boneToModify {boneName=dyn_L_backCoat_MSTR}
    rotation {roll=12.094}
    translation {y=-3.39}
   }
   n18 = "Transform (Modify) Bone - Bone: dyn_L_sideCoat_MSTR"(ComponentPose=@n17, Alpha=@n16, Translation="3.618,-0.000,0.000", Rotation="9.844,-25.400,0.000", Scale="1,1,1") {
    rotationMode=BMM_Additive
    translationMode=BMM_Additive
    alphaBoolBlend {blendOption=Linear}
    boneToModify {boneName=dyn_L_sideCoat_MSTR}
    rotation=<rot 9.844 -25.4 0>
    translation {x=3.618}
   }
   n19 = RigidBody(ComponentPose=@n18, ExternalForce=Get GravityOverride()) {
    bClampLinearTranslationLimitToRefPose=true
    overridePhysicsAsset=&/F_MED_MadameMoth_Posh_NPC_Physics
    simulationSpace=BaseBoneSpace
    alphaBoolBlend {blendOption=Linear}
    baseBoneRef {boneName=pelvis}
   }
   n20 = "Transform (Modify) Bone - Bone: dyn_R_sideCoat_MSTR"(ComponentPose=@n19, Alpha=Get Not Crouching() -> (Not Crouching:double), Translation="-1,0,0", Rotation="0,0,0", Scale="1,1,1") {
    rotationMode=BMM_Additive
    translationMode=BMM_Additive
    alphaBoolBlend {blendOption=Linear}
    boneToModify {boneName=dyn_R_sideCoat_MSTR}
    translation {x=-1}
   }
   n21 = "Transform (Modify) Bone - Bone: dyn_R_frontCoat_MSTR"(ComponentPose=@n20, Alpha=Get Not Crouching(), Translation="0,0,0", Rotation="-11,0,1", Scale="1,1,1") {
    rotationMode=BMM_Additive
    translationMode=BMM_Additive
    alphaBoolBlend {blendOption=Linear}
    boneToModify {boneName=dyn_R_frontCoat_MSTR}
    rotation {pitch=-11, roll=1}
   }
   n22 = "Transform (Modify) Bone - Bone: dyn_R_backCoat_MSTR"(ComponentPose=@n21, Alpha=Get Not Crouching(), Translation="0,-2,0", Rotation="0,0,10", Scale="1,1,1") {
    rotationMode=BMM_Additive
    translationMode=BMM_Additive
    alphaBoolBlend {blendOption=Linear}
    boneToModify {boneName=dyn_R_backCoat_MSTR}
    rotation {roll=10}
    translation {y=-2}
   }
   n23 = "Transform (Modify) Bone - Bone: dyn_R_sideCoat_MSTR"(ComponentPose=@n22, Alpha=Get crouching alpha(), Translation="0,0,0", Rotation="-3.534,54.560,11.693", Scale="1,1,1") {
    rotationMode=BMM_Additive
    translationMode=BMM_Additive
    alphaBoolBlend {blendOption=Linear}
    boneToModify {boneName=dyn_R_sideCoat_MSTR}
    rotation=<rot -3.534 54.56 11.693>
   }
   n24 = "Transform (Modify) Bone - Bone: dyn_R_frontCoat_MSTR"(ComponentPose=@n23, Alpha=Get crouching alpha(), Translation="-8.388,0.000,1.744", Rotation="4.331,29.006,-0.000", Scale="1,1,1") {
    rotationMode=BMM_Additive
    translationMode=BMM_Additive
    alphaBoolBlend {blendOption=Linear}
    boneToModify {boneName=dyn_R_frontCoat_MSTR}
    rotation=<rot 4.331 29.006 0>
    translation {x=-8.388, z=1.744}
   }
   n25 = "Transform (Modify) Bone - Bone: dyn_R_backCoat_MSTR"(ComponentPose=@n24, Alpha=Get crouching alpha(), Translation="0,0,0", Rotation="-2.320,4.408,11.051", Scale="1,1,1") {
    rotationMode=BMM_Additive
    translationMode=BMM_Additive
    alphaBoolBlend {blendOption=Linear}
    boneToModify {boneName=dyn_R_backCoat_MSTR}
    rotation=<rot -2.32 4.408 11.051>
   }
   n26 = Get Not Crouching()
   n27 = "Transform (Modify) Bone - Bone: dyn_L_backCoat_MSTR"(ComponentPose=@n25, Alpha=@n26, Translation="0.000,-3.659,-0.000", Rotation="-0.000,0.000,7.386", Scale="1,1,1") {
    rotationMode=BMM_Additive
    translationMode=BMM_Additive
    alphaBoolBlend {blendOption=Linear}
    boneToModify {boneName=dyn_L_backCoat_MSTR}
    rotation {roll=7.386}
    translation {y=-3.659}
   }
   n28 = "Transform (Modify) Bone - Bone: dyn_L_sideCoat_MSTR"(ComponentPose=@n27, Alpha=@n26, Translation="0.000,1.669,-0.000", Rotation="16.123,-1.373,2.596", Scale="1,1,1") {
    rotationMode=BMM_Additive
    translationMode=BMM_Additive
    alphaBoolBlend {blendOption=Linear}
    boneToModify {boneName=dyn_L_sideCoat_MSTR}
    rotation=<rot 16.123 -1.373 2.596>
    translation {y=1.669}
   }
   Output Pose(Result=Component To Local(ComponentPose=@n28)) {layerGroup=None}
  ]}
 EventGraph {nodes [
   Event BlueprintUpdateAnimation()
   > n1 = Gravity Override State Function DEPRECATED(CharacterPartAnimInstance=Self-Reference(), GameTimeInSeconds=GetGameTimeInSeconds(), GravityOverrideParams=Get Gravity Override Params())
   > Set GravityOverride(GravityOverride=@n1)
   Event OnBeginCrouching()
   > Set crouching alpha(crouching alpha:double=1) -> (Output_Get:double)
   > Set Not Crouching(Not Crouching:double=0) -> (Output_Get:double)
   Event OnEndCrouching()
   > Set crouching alpha(crouching alpha=0)
   > Set Not Crouching(Not Crouching=1)
   TryGetPawnOwner()
  ]}
]
variables [
 crouching alpha {type=number}
 Gravity Override Params {type=GravityOverrideParamsStruct}
 GravityOverride {type=Vector}
 Not Crouching {type=number, default=1}
]
